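; Photo mode camera modifier for the back-facing mode, which positions the camera relative to the player.
; GetCameraUpdatedLocation adds offsets along the player's forward and up vectors, scaled by ForwardOffset and UpOffset, to the camera base location.
; The incoming location and rotation are ignored, so the result follows the player's orientation, and this function computes position only.
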(fn GetCameraUpdatedLocation
  (params (InLocation :type Vector :default "0, 0, 0") (inRotation :type Rotator :default "0, 0, 0"))
  (return (Utilities|Operators|Add (CallFunction|GetCameraBaseLocation (Variables|Default|GetPlayer)) (* (Transformation|GetActorForwardVector (Variables|Default|GetPlayer)) (Variables|Default|GetForwardOffset)) (* (Transformation|GetActorUpVector (Variables|Default|GetPlayer)) (Variables|Default|GetUpOffset)))))
