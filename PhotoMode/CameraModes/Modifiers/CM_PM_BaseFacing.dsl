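; Base photo mode camera modifier for facing-style cameras, caching the controlled Fort player pawn when added to a camera.
; Each update blends view location toward GetCameraUpdatedLocation by modifier alpha, leaving rotation and FOV untouched; that function is an empty stub here.
; GetCameraBaseLocation derives an origin from actor position and socket up-vector offsets, or capsule height without a valid mesh.
; UpdateFacingValues stores positional, yaw and pitch offsets plus default FOV from facing data.
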(event EventBlueprintAddedToCamera
  (params (Camera :type Player%20Camera%20Manager%20Object%20Reference))
  (Variables|Default|SetCamera Camera)
  (bind _ascontroller (Utilities|Casting|CastToController (Actor|GetOwner Camera))
    (:then
      (Variables|Default|SetPlayer (Utilities|Casting|CastToFortPlayerPawn (Pawn|GetControlledPawn _ascontroller))))
    (:CastFailed
      (Variables|Default|SetPlayer 0))))

(fn BlueprintModifyCamera
  (params (DeltaTime :type Float%20%28single-precision%29) (ViewLocation :type Vector :default "0, 0, 0") (ViewRotation :type Rotator :default "0, 0, 0") (FOV :type Float%20%28single-precision%29))
  (bind _cameralocation (CallFunction|GetCameraUpdatedLocation ViewLocation ViewRotation))
  (return (Math|Vector|Lerp(Vector) ViewLocation _cameralocation (Variables|CameraModifier|GetAlpha)) ViewRotation FOV))

(fn GetCameraBaseLocation
  (params (PlayerPawn :type Fort%20Player%20Pawn%20Object%20Reference))
  (bind _output_get (LocalVariables|SetSkeletalMeshComp (Variables|/Script/Engine.Character|GetMesh PlayerPawn)))
  (Utilities|IsValid _output_get
    (:Is%20Valid
      (return (+ (- (Transformation|GetActorLocation PlayerPawn) (* (Transformation|GetActorUpVector PlayerPawn) (Variables|Default|GetSocketDownOffset))) (* (Math|Vector|GetUpVector (Transformation|GetSocketRotation _output_get (Variables|Default|GetCameraOriginSocketName))) (Variables|Default|GetSocketUpOffset)))))
    (:Is%20Not%20Valid
      (return (+ (Transformation|GetActorLocation (Variables|Default|GetPlayer)) (Math|Vector|MakeVector 0.0 0.0 (* (Components|Capsule|GetScaledCapsuleHalfHeight (Variables|/Script/Engine.Character|GetCapsuleComponent (Variables|Default|GetPlayer))) 0.3)))))))

(fn GetCameraUpdatedLocation
  (params (InLocation :type Vector) (inRotation :type Rotator)))

(fn UpdateFacingValues
  (params (InNewValues_ForwardOffset :type Float%20%28single-precision%29) (InNewValues_RightOffset :type Float%20%28single-precision%29) (InNewValues_UpOffset :type Float%20%28single-precision%29) (InNewValues_YawOffset :type Float%20%28single-precision%29) (InNewValues_PitchOffset :type Float%20%28single-precision%29) (InNewValues_MinPitch :type Float%20%28single-precision%29) (InNewValues_MaxPitch :type Float%20%28single-precision%29) (InNewValues_DefaultFOV :type Float%20%28single-precision%29) (InNewValues_MinMaxFOV :type Vector%202D%20Structure) (InNewValues_Montage :type Anim%20Montage%20Object%20Reference) (InNewValues_MontageSection :type Name))
  (bind _innewvalues (Variables|GetInNewValues))
  (bind (_forwardoffset _rightoffset _upoffset) (Utilities|Struct|BreakPhotoCameraFacingData _innewvalues))
  (Utilities|FlowControl|Sequence
    (:then_0
      (Variables|Default|SetForwardOffset _forwardoffset)
      (Variables|Default|SetRightOffset _rightoffset)
      (Variables|Default|SetUpOffset _upoffset))
    (:then_1
      (Variables|Default|SetPitchOffset InNewValues_PitchOffset)
      (Variables|Default|SetYawOffset InNewValues_YawOffset)
      (Variables|Default|SetDefault%20FOV InNewValues_DefaultFOV))))
