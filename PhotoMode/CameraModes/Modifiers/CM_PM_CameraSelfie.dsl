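; Photo mode camera modifier that positions the view relative to the player's head for selfie shots.
; With the selfie facing gameplay tag, view pitch drives sine and cosine offsets, arcing the camera vertically around the head.
; Otherwise it applies fixed forward and up offsets, leaving rotation and FOV unchanged either way.
; UpdateFacingValues supplies offsets from facing data; head location uses the mesh socket, or capsule height without a mesh.
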(event EventBlueprintAddedToCamera
  (params (Camera :type Player%20Camera%20Manager%20Object%20Reference))
  (Variables|Default|SetCamera Camera)
  (bind _ascontroller (Utilities|Casting|CastToController (Actor|GetOwner Camera))
    (:then
      (Variables|Default|SetPlayer (Utilities|Casting|CastToFortPlayerPawn (Pawn|GetControlledPawn _ascontroller))))
    (:CastFailed
      (Variables|Default|SetPlayer 0))))

(fn BlueprintModifyCamera
  (params (DeltaTime :type Float%20%28single-precision%29) (ViewLocation :type Vector :default "0, 0, 0") (ViewRotation :type Rotator :default "0, 0, 0") (FOV :type Float%20%28single-precision%29))
  (bind (_currentselfieoffset_x _currentselfieoffset_y _currentselfieoffset_z) (Math|Vector|BreakVector (Variables|GetCurrentSelfieOffset_X)))
  (bind (_viewrotation_roll _viewrotation_pitch) (Math|Rotator|BreakRotator (Variables|GetViewRotation_Roll)))
  (bind _player (Variables|Default|GetPlayer))
  (if (GameplayTags|HasMatchingGameplayTag _player "(TagName=\"Ability.PhotoMode.Camera.Facing.Selfie\")")
    (Utilities|FlowControl|Sequence
      (:then_0
        (Variables|SetCurrentSelfieOffset_X (Math|Vector|MakeVector (Utilities|Operators|Multiply (Variables|Default|GetForwardOffset) (Math|Trig|Cos(Degrees) (Math|Float|ClampAngle (* _viewrotation_pitch -1.0) -9e+01 9e+01)) -1.0) (Variables|Default|GetRightOffset) (+ (Variables|Default|GetUpOffset) (* (Variables|Default|GetForwardOffset) (Math|Trig|Sin(Degrees) (Math|Float|ClampAngle (* _viewrotation_pitch -1.0) -9e+01 9e+01)))))))
      (:then_1
        (bind _output_get (Variables|Default|SetLastSelfieCamLocation (+ (CallFunction|GetHeadLocation (Variables|Default|GetPlayer)) (Utilities|Operators|Add (* (Transformation|GetActorForwardVector (Variables|Default|GetPlayer)) _currentselfieoffset_x) (* (Transformation|GetActorRightVector (Variables|Default|GetPlayer)) _currentselfieoffset_y) (Math|Vector|MakeVector 0.0 0.0 _currentselfieoffset_z)))))
        (return _output_get (Variables|GetViewRotation) (Variables|GetFOV))))
    (else
      (return (Utilities|Operators|Add (CallFunction|GetHeadLocation _player) (* (Transformation|GetActorForwardVector (Variables|Default|GetPlayer)) (Variables|Default|GetForwardOffset)) (* (Transformation|GetActorUpVector (Variables|Default|GetPlayer)) (Variables|Default|GetUpOffset))) (Variables|GetViewRotation) (Variables|GetFOV)))))

(fn GetHeadLocation
  (params (PlayerPawn :type Fort%20Player%20Pawn%20Object%20Reference))
  (bind _output_get (LocalVariables|SetSkeletalMeshComp (Variables|/Script/Engine.Character|GetMesh PlayerPawn)))
  (Utilities|IsValid _output_get
    (:Is%20Valid
      (return (Transformation|GetSocketLocation _output_get "head")))
    (:Is%20Not%20Valid
      (return (+ (Transformation|GetActorLocation (Variables|Default|GetPlayer)) (Math|Vector|MakeVector 0.0 0.0 (* (Components|Capsule|GetScaledCapsuleHalfHeight (Variables|/Script/Engine.Character|GetCapsuleComponent (Variables|Default|GetPlayer))) 0.8)))))))

(fn UpdateFacingValues
  (params (InNewValues_ForwardOffset :type Float%20%28single-precision%29) (InNewValues_RightOffset :type Float%20%28single-precision%29) (InNewValues_UpOffset :type Float%20%28single-precision%29) (InNewValues_YawOffset :type Float%20%28single-precision%29) (InNewValues_PitchOffset :type Float%20%28single-precision%29) (InNewValues_MinPitch :type Float%20%28single-precision%29) (InNewValues_MaxPitch :type Float%20%28single-precision%29) (InNewValues_DefaultFOV :type Float%20%28single-precision%29) (InNewValues_MinMaxFOV :type Vector%202D%20Structure) (InNewValues_Montage :type Anim%20Montage%20Object%20Reference) (InNewValues_MontageSection :type Name))
  (bind _innewvalues (Variables|GetInNewValues))
  (bind (_forwardoffset _rightoffset _upoffset) (Utilities|Struct|BreakPhotoCameraFacingData _innewvalues))
  (Utilities|FlowControl|Sequence
    (:then_0
      (Variables|Default|SetForwardOffset _forwardoffset)
      (Variables|Default|SetRightOffset _rightoffset)
      (Variables|Default|SetUpOffset _upoffset))
    (:then_1
      (Variables|Default|SetPitchOffset InNewValues_PitchOffset)
      (Variables|Default|SetYawOffset InNewValues_YawOffset))))
