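; Photo Mode base camera modifier that zooms by changing FOV over time, in the direction given by the sign of ZoomSpeed.
; When added, it picks a target FOV limit from the facing data on the player pawn's anim component.
; Each update steps FOV toward that limit, clamps it, and writes it back to the current custom camera mode.
; PauseZoom zeroes the current speed, while ResumeZoom restores it with a multiplier for the most recently used input device type.
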(event EventBlueprintAddedToCamera
  (params (Camera :type Player%20Camera%20Manager%20Object%20Reference))
  (bind _asfort_player_camera_base (Utilities|Casting|CastToFortPlayerCameraBase Camera))
  (Variables|Default|SetCameraManager _asfort_player_camera_base)
  (Variables|Default|SetCurrentZoomSpeed (Variables|Default|GetZoomSpeed))
  (CallFunction|UpdateTargetFOV))

(fn BlueprintModifyCamera
  (params (DeltaTime :type Float%20%28single-precision%29) (ViewLocation :type Vector :default "0, 0, 0") (ViewRotation :type Rotator :default "0, 0, 0") (FOV :type Float%20%28single-precision%29))
  (bind _zoomspeed (Variables|Default|GetZoomSpeed))
  (bind _returnvalue (>= _zoomspeed 0))
  (bind _targetfov (Variables|Default|GetTargetFOV))
  (bind _deltatime (Variables|GetDeltaTime))
  (bind _currentzoomspeed (Variables|Default|GetCurrentZoomSpeed))
  (bind _returnvalue_2 (* _currentzoomspeed _deltatime))
  (bind _returnvalue_3 (+ FOV _returnvalue_2))
  (bind _returnvalue_4 (Math|Float|Max(Float) _returnvalue_3 _targetfov))
  (bind _targetfov_2 (Variables|Default|GetTargetFOV))
  (bind _returnvalue_5 (Math|Float|Min(Float) _returnvalue_3 _targetfov_2))
  (bind _returnvalue_6 (Math|Float|SelectFloat _returnvalue_5 _returnvalue_4 _returnvalue))
  (if (== (Variables|Default|GetCurrentZoomSpeed) 0)
    (return (Variables|GetViewLocation) (Variables|GetViewRotation) (Variables|GetFOV))
    (else
      (bind _asfort_camera_mode_custom (Utilities|Casting|CastToFortCameraMode_Custom (Class|FortPlayerCameraBase|GetCurrentCameraMode (Variables|Default|GetCameraManager))))
      (Variables|/Script/FortniteGame.FortCameraMode_ThirdPerson|SetFOV _returnvalue_6 _asfort_camera_mode_custom)
      (return (Variables|GetViewLocation) (Variables|GetViewRotation) _returnvalue_6))))

(fn PauseZoom
  (Variables|Default|SetCurrentZoomSpeed 0.0))

(fn ResumeZoom
  (bind _returnvalue (EngineSubsystems|GetInputDeviceSubsystem))
  (bind (_returnvalue_inputclassname _returnvalue_hardwaredeviceidentifier _returnvalue_primarydevicetype) (Utilities|Struct|BreakHardwareDeviceIdentifier (InputDevices|GetMostRecentlyUsedHardwareDevice :self _returnvalue)))
  (bind _zoom_speed_multiplier_by_device (Variables|Default|GetZoom%20Speed%20Multiplier%20by%20Device))
  (bind (_value _returnvalue_2) (Utilities|Map|Find _zoom_speed_multiplier_by_device _returnvalue_primarydevicetype))
  (Variables|Default|SetCurrentZoomSpeed (Variables|Default|GetZoomSpeed))
  (Utilities|IsValid _returnvalue
    (:Is%20Valid
      (if _returnvalue_2
        (Variables|Default|SetCurrentZoomSpeed (* (Variables|Default|GetCurrentZoomSpeed) _value))))))

(fn UpdateTargetFOV
  (bind _output_get (LocalVariables|SetFacingData (FacingData|GetCurrentFacingData (Actor|GetComponentbyClass (Pawn|GetControlledPawn (Game|Player|GetOwningPlayerController (Variables|Default|GetCameraManager))) "/PhotoMode/Blueprints/B_PM_PlayerPawnAnim_Component.B_PM_PlayerPawnAnim_Component_C"))))
  (Variables|Default|SetTargetFOV (Math|Float|SelectFloat (.Y (Math|Vector2D|BreakVector2D (Utilities|Struct|BreakPhotoCameraFacingData _output_get))) (Math|Vector2D|BreakVector2D (Utilities|Struct|BreakPhotoCameraFacingData _output_get)) (> (Variables|Default|GetZoomSpeed) 0))))
