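; Photo mode camera modifier for the selfie view, which computes the camera location relative to the player.
; Negated, clamped camera pitch swings the forward offset through a vertical arc using sine and cosine, combined with fixed right and up offsets.
; The resulting offset follows the player's forward and right vectors and world Z from the camera base location, ignoring the input location.
; The result is cached as the last selfie camera location and returned.
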(fn GetCameraUpdatedLocation
  (params (InLocation :type Vector :default "0, 0, 0") (inRotation :type Rotator :default "0, 0, 0"))
  (bind (_currentselfieoffset_x _currentselfieoffset_y _currentselfieoffset_z) (Math|Vector|BreakVector (Variables|GetCurrentSelfieOffset_X)))
  (Utilities|FlowControl|Sequence
    (:then_0
      (Variables|SetCurrentSelfieOffset_X (Math|Vector|MakeVector (Utilities|Operators|Multiply (Variables|Default|GetForwardOffset) (Math|Trig|Cos(Degrees) (Math|Float|ClampAngle (* (.pitch inRotation) -1.0) -9e+01 9e+01)) -1.0) (Variables|Default|GetRightOffset) (+ (Variables|Default|GetUpOffset) (* (Variables|Default|GetForwardOffset) (Math|Trig|Sin(Degrees) (Math|Float|ClampAngle (* (.pitch inRotation) -1.0) -9e+01 9e+01)))))))
    (:then_1
      (bind _output_get (Variables|Default|SetLastSelfieCamLocation (+ (CallFunction|GetCameraBaseLocation (Variables|Default|GetPlayer)) (Utilities|Operators|Add (* (Transformation|GetActorForwardVector (Variables|Default|GetPlayer)) _currentselfieoffset_x) (* (Transformation|GetActorRightVector (Variables|Default|GetPlayer)) _currentselfieoffset_y) (Math|Vector|MakeVector 0.0 0.0 _currentselfieoffset_z)))))
      (return _output_get))))
